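# Sound cue for a distant meteor falling through the sky, part of the volcano ambient sounds.
# A mixer layers two falling-meteor waves, each shaped by its own volume envelope: the first swells quickly and fades, while the second starts after a fixed delay and builds slowly before cutting off, forming one staged travel sound.
# The cue plays boosted through far volcano-shot attenuation and a Nexus transition sound class.
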SoundCue /Figment_S08/Athena/Sounds/Ambient/Volcano/Amb_Meteor_Travel_01_Cue
root=Mixer_0
nodes [
 {id=Delay_0, children [Enveloper_3], properties {delayMax=5, delayMin=5}}
 {
  id=Enveloper_2
  children [WavePlayer_0]
  properties {
   pitchCurve {editorCurveData {keys [{time=0}]}}
   volumeCurve {editorCurveData {keys [<at 0 -0.0107>,<at 0.1871 0.2635>,<at 0.2528 0.9992>,<at 2.9942 0.9863>,<at 5.0146 0.0849>]}}
  }
 }
 {
  id=Enveloper_3
  children [WavePlayer_3]
  properties {
   pitchCurve {editorCurveData {keys [{time=0}]}}
   volumeCurve {editorCurveData {keys [<at 0 -0.0107>,<at 0.9519 0.193>,<at 2.7229 0.488>,<at 4.0517 0.9866>,<at 5.7037 1.0059>,<at 6.0923 0.0043>]}}
  }
 }
 {id=Mixer_0, children [Enveloper_2,Delay_0], properties {inputVolume [1,1]}}
 {id=WavePlayer_0, properties {soundWaveAssetPtr=/Figment_S08/Sounds/Fort_GamePlay_Sounds/Meteor/Meteor_Falling_Distant_01}}
 {id=WavePlayer_3, properties {soundWaveAssetPtr=/Figment_S08/Sounds/Fort_GamePlay_Sounds/Meteor/Meteor_Falling_Distant_02}}
]
properties {
 attenuationSettings=/Figment_S08/Sounds/Attenuation/Fort_Volcano_Shoot_Far_Att
 maxDistance=2097152
 soundClassObject=/Figment_S08/Sounds/FortSoundClassesAndModes/Fort_Nexus_Transition
 volumeMultiplier=2
 concurrencyOverrides {maxCount=1, resolutionRule=PreventNew}
}
