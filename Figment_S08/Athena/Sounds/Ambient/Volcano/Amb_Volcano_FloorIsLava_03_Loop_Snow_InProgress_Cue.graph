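# Ambient looping sound cue for the Floor Is Lava volcano, named as an in-progress snow variant.
# It plays the reused safe-zone storm high loop through two nested mixers whose second inputs are empty, with the inner one cutting its level.
# It routes to the Nexus ambient sound class with a volume boost and a very large max distance.
# Concurrency allows a single instance and blocks new ones.
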SoundCue /Figment_S08/Athena/Sounds/Ambient/Volcano/Amb_Volcano_FloorIsLava_03_Loop_Snow_InProgress_Cue
root=Mixer_0
nodes [
 {id=Mixer_0, children [Mixer_1,null], properties {inputVolume [1,1.5]}}
 {id=Mixer_1, children [WavePlayer_7,null], properties {inputVolume [0.25,1.5]}}
 {id=WavePlayer_7, properties {bLooping=true, soundWaveAssetPtr=/Figment_S08/Athena/Sounds/Shield/SafeZone_Storm_Loop_Hgih}}
]
properties {
 maxDistance=2097152
 soundClassObject=/Figment_S08/Sounds/FortSoundClassesAndModes/Fort_Nexus_Ambient
 volumeMultiplier=1.75
 concurrencyOverrides {maxCount=1, resolutionRule=PreventNew}
}
