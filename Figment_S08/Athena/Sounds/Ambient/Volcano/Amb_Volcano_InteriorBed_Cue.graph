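# Ambient sound cue for the volcano interior bed in Athena.
# A mixer layers a looping lava lake wave with a Random node that picks one of two equally weighted, looping volcano exterior waves.
# It uses Athena ambient attenuation and a duck-outdoors ambient sound class, with a boosted volume multiplier.
# Concurrency allows a single instance and prevents new ones from starting.
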SoundCue /Figment_S08/Athena/Sounds/Ambient/Volcano/Amb_Volcano_InteriorBed_Cue
root=Mixer_1
nodes [
 {id=Mixer_1, children [WavePlayer_18, Random_0], properties {inputVolume [1,1]}}
 {id=Random_0, children [WavePlayer_4,WavePlayer_5], properties {weights [1,1]}}
 {id=WavePlayer_18, properties {bLooping=true, soundWaveAssetPtr=&/Amb_Volcano_Lava_LakeLoop_01}}
 {id=WavePlayer_4, properties {bLooping=true, soundWaveAssetPtr=&/Amb_Volcano_ExteriorLoop_01}}
 {id=WavePlayer_5, properties {bLooping=true, soundWaveAssetPtr=&/Amb_Volcano_ExteriorLoop_02}}
]
properties {
 attenuationSettings=&Fort_Ambient_Athena_Att
 maxDistance=2097152
 soundClassObject=/Game/Sounds/FortSoundClassesAndModes/Fort_Ambient_DuckOutdoors_Athena_PSM
 volumeMultiplier=2
 concurrencyOverrides {maxCount=1, resolutionRule=PreventNew}
}
