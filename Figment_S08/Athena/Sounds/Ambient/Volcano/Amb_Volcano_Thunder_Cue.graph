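# Sound cue for ambient volcano thunder that plays as intermittent one-shots on an endless loop.
# A looping node wraps a randomized delay of tens of seconds, which feeds an equally weighted random pick among four thunder wave variations.
# It plays with boosted volume through ambient attenuation and an outdoor-ducking ambient sound class, with concurrency limited to one instance that blocks new ones.
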SoundCue /Figment_S08/Athena/Sounds/Ambient/Volcano/Amb_Volcano_Thunder_Cue
root=Looping_0
nodes [
 {id=Delay_0, children [Random_1], properties {delayMax=30, delayMin=10}}
 {id=Looping_0, children [Delay_0]}
 {id=Random_1, children [WavePlayer_0,WavePlayer_1,WavePlayer_2,WavePlayer_3], properties {weights [1,1,1,1]}}
 {id=WavePlayer_0, properties {soundWaveAssetPtr=&/Amb_Volcano_Thunder_01}}
 {id=WavePlayer_1, properties {soundWaveAssetPtr=/FNGameplayCues/Athena/Sounds/Ambient/Volcano/Amb_Volcano_Thunder_02}}
 {id=WavePlayer_2, properties {soundWaveAssetPtr=/FNGameplayCues/Athena/Sounds/Ambient/Volcano/Amb_Volcano_Thunder_03}}
 {id=WavePlayer_3, properties {soundWaveAssetPtr=&/Amb_Volcano_Thunder_04}}
]
properties {
 attenuationSettings=&Fort_Ambient_Athena_Att
 maxDistance=2097152
 soundClassObject=/Game/Sounds/FortSoundClassesAndModes/Fort_Ambient_DuckOutdoors_Athena_PSM
 volumeMultiplier=2
 concurrencyOverrides {maxCount=1, resolutionRule=PreventNew}
}
