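# Sound cue for a volcano ambient meteor explosion, staging an approach, impact and distant blast with far volcano attenuation.
# A root mixer plays the near-landing wave immediately, then delayed layers: impact with sonic boom, random sub-impact, and distant explosion with boosted blast.
# Volume envelopes fade the impact out and the first distant explosion in; another distant explosion follows a minute later.
# Concurrency allows one instance and rejects new plays while it runs.
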SoundCue /Figment_S08/Athena/Sounds/Ambient/Volcano/Explo_Meteor_Exlpo_01_Cue
root=Mixer_5
nodes [
 {id=Delay_10, children [Modulator_7], properties {delayMax=2, delayMin=2}}
 {id=Delay_12, children [Modulator_8], properties {delayMax=2, delayMin=2}}
 {id=Delay_3, children [WavePlayer_31], properties {delayMax=60, delayMin=60}}
 {id=Delay_8, children [Mixer_4], properties {delayMax=1.25, delayMin=1.25}}
 {id=Delay_9, children [Random_3], properties {delayMax=1.5, delayMin=1.5}}
 {
  id=Enveloper_5
  children [WavePlayer_24]
  properties {pitchCurve {editorCurveData {keys [{time=0}]}}, volumeCurve {editorCurveData {keys [<at 0 -0.0005>,<at 0.0841 0.5422>,<at 0.1779 1.015>]}}}
 }
 {
  id=Enveloper_6
  children [WavePlayer_30]
  properties {
   pitchCurve {editorCurveData {keys [{time=0}]}}
   volumeCurve {editorCurveData {keys [<at 0 1>,<at 0.9049 1.0143>,<at 1.1681 0.4706>,<at 1.5211 0.1326>,<at 2.0217 0.0004>]}}
  }
 }
 {id=Mixer_4, children [Modulator_9,Modulator_6], properties {inputVolume [1,1]}}
 {id=Mixer_5, children [WavePlayer_32, Delay_8, Delay_9, Delay_10, Delay_12, Delay_3], properties {inputVolume [1,1,1,1,1.5,1]}}
 {id=Modulator_6, children [WavePlayer_29]}
 {id=Modulator_7, children [Enveloper_5]}
 {id=Modulator_8, children [WavePlayer_26]}
 {id=Modulator_9, children [Enveloper_6]}
 {id=Random_3, children [WavePlayer_28, WavePlayer_27], properties {weights [1,1]}}
 {id=WavePlayer_24, properties {soundWaveAssetPtr=/Game/Sounds/Fort_GamePlay_Sounds/Explosives/fort_distant_explosion_sharp}}
 {id=WavePlayer_26, properties {soundWaveAssetPtr=&/Explo_Blast_Distant_01}}
 {id=WavePlayer_27, properties {soundWaveAssetPtr=/Figment_S08/Sounds/Fort_GamePlay_Sounds/Meteor/Meteor_SubImpact_01}}
 {id=WavePlayer_28, properties {soundWaveAssetPtr=/Figment_S08/Sounds/Fort_GamePlay_Sounds/Meteor/Meteor_SubImpact_03}}
 {id=WavePlayer_29, properties {soundWaveAssetPtr=/Figment_S08/Sounds/Fort_GamePlay_Sounds/Meteor/Meteor_Sonic_Boom_02}}
 {id=WavePlayer_30, properties {soundWaveAssetPtr=/Figment_S08/Sounds/Fort_GamePlay_Sounds/Meteor/Meteor_Impact_02}}
 {id=WavePlayer_31, properties {soundWaveAssetPtr=&fort_distant_explosion_sharp}}
 {id=WavePlayer_32, properties {soundWaveAssetPtr=/Game/Sounds/Fort_GamePlay_Sounds/Meteor/Meteor_Almost_Landed_01}}
]
properties {
 attenuationSettings=/Figment_S08/Sounds/Attenuation/Fort_Volcano_Shoot_Far_Att
 maxDistance=2097152
 soundClassObject=/Figment_S08/Sounds/FortSoundClassesAndModes/Fort_Nexus_Transition
 volumeMultiplier=2
 concurrencyOverrides {maxCount=1, resolutionRule=PreventNew}
}
