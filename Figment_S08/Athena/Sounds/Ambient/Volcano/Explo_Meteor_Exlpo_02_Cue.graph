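# Sound cue for a volcano ambient meteor explosion, using far volcano-shot attenuation.
# It mixes a meteor impact and a sonic boom, each through a modulator, with a short-delayed sub-impact chosen at random from two waves.
# An enveloper fades the impact layer out after its first second.
# Its concurrency override allows one instance at a time and blocks new ones while it plays.
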SoundCue /Figment_S08/Athena/Sounds/Ambient/Volcano/Explo_Meteor_Exlpo_02_Cue
root=Mixer_3
nodes [
 {id=Delay_1, children [Random_4], properties {delayMax=0.2, delayMin=0.2}}
 {
  id=Enveloper_2
  children [WavePlayer_3]
  properties {
   pitchCurve {editorCurveData {keys [{time=0}]}}
   volumeCurve {editorCurveData {keys [<at 0 1>,<at 0.9049 1.0143>,<at 1.1681 0.4706>,<at 1.5211 0.1326>,<at 2.0217 0.0004>]}}
  }
 }
 {id=Mixer_0, children [Modulator_3,Modulator_0], properties {inputVolume [1,1]}}
 {id=Mixer_3, children [Mixer_0,Delay_1], properties {inputVolume [1,1]}}
 {id=Modulator_0, children [WavePlayer_9]}
 {id=Modulator_3, children [Enveloper_2]}
 {id=Random_4, children [WavePlayer_4, WavePlayer_10], properties {weights [1,1]}}
 {id=WavePlayer_10, properties {soundWaveAssetPtr=/Figment_S08/Sounds/Fort_GamePlay_Sounds/Meteor/Meteor_SubImpact_01}}
 {id=WavePlayer_3, properties {soundWaveAssetPtr=/Figment_S08/Sounds/Fort_GamePlay_Sounds/Meteor/Meteor_Impact_02}}
 {id=WavePlayer_4, properties {soundWaveAssetPtr=/Figment_S08/Sounds/Fort_GamePlay_Sounds/Meteor/Meteor_SubImpact_03}}
 {id=WavePlayer_9, properties {soundWaveAssetPtr=/Figment_S08/Sounds/Fort_GamePlay_Sounds/Meteor/Meteor_Sonic_Boom_02}}
]
properties {
 attenuationSettings=/Figment_S08/Sounds/Attenuation/Fort_Volcano_Shoot_Far_Att
 maxDistance=2097152
 soundClassObject=/Figment_S08/Sounds/FortSoundClassesAndModes/Fort_Nexus_Transition
 volumeMultiplier=1.5
 concurrencyOverrides {maxCount=1, resolutionRule=PreventNew}
}
