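# Sound cue that plays a timed multi-explosion meteor impact sequence for the volcano ambience, limited to one instance with far volcano-shot attenuation.
# A root mixer layers a random near-landing sound with delayed branches of impacts, sonic booms and sub-impacts over several seconds.
# A last nested mixer adds an eruption, an enveloped distant explosion and a delayed impact stack ending in a shockwave layer.
# Random wave picks and pitch modulators vary the reused samples.
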SoundCue /Figment_S08/Athena/Sounds/Ambient/Volcano/Explo_Meteor_Exlpo_Multi_Cue
root=Mixer_3
nodes [
 {id=Delay_0, children [Mixer_0], properties {delayMax=1, delayMin=1}}
 {id=Delay_1, children [Modulator_11], properties {delayMax=1, delayMin=1}}
 {id=Delay_10, children [Modulator_7], properties {delayMax=0.75, delayMin=0.75}}
 {id=Delay_2, children [Modulator_2], properties {delayMax=3.5, delayMin=3.5}}
 {id=Delay_3, children [Modulator_1], properties {delayMax=3, delayMin=3}}
 {id=Delay_4, children [Modulator_0], properties {delayMax=4.25, delayMin=4.25}}
 {id=Delay_5, children [Mixer_5], properties {delayMax=6.4, delayMin=6.4}}
 {id=Delay_8, children [Modulator_5], properties {delayMax=0.25, delayMin=0.25}}
 {id=Delay_9, children [Modulator_4], properties {delayMax=0.9, delayMin=0.9}}
 {
  id=Enveloper_0
  children [WavePlayer_27]
  properties {pitchCurve {editorCurveData {keys [{time=0}]}}, volumeCurve {editorCurveData {keys [<at 0 -0.0005>,<at 0.1183 0.3231>,<at 0.1779 1.015>]}}}
 }
 {id=Mixer_0, children [Random_0,Modulator_9], properties {inputVolume [1,1]}}
 {id=Mixer_1, children [Modulator_8,Modulator_10], properties {inputVolume [1,1]}}
 {id=Mixer_3, children [Random_1,Delay_0,Delay_1,Delay_3,Delay_2,Delay_4,Delay_5], properties {inputVolume [0.3,1,1,1,1,1.1,1.1]}}
 {id=Mixer_4, children [Modulator_12, WavePlayer_24, Modulator_3, Delay_10], properties {inputVolume [1.15,1,1,1]}}
 {id=Mixer_5, children [Modulator_6,Delay_8,WavePlayer_1,Delay_9], properties {inputVolume [1,1,1,1]}}
 {id=Modulator_0, children [WavePlayer_25]}
 {id=Modulator_1, children [Mixer_1]}
 {id=Modulator_10, children [WavePlayer_14], properties {pitchMax=0.97}}
 {id=Modulator_11, children [WavePlayer_4]}
 {id=Modulator_12, children [Random_3], properties {pitchMax=0.9, pitchMin=0.9}}
 {id=Modulator_2, children [WavePlayer_11]}
 {id=Modulator_3, children [Random_7]}
 {id=Modulator_4, children [Mixer_4], properties {pitchMax=1.1, pitchMin=1.1, volumeMax=1, volumeMin=1}}
 {id=Modulator_5, children [Enveloper_0]}
 {id=Modulator_6, children [Random_14], properties {pitchMax=1.2, pitchMin=1.1}}
 {id=Modulator_7, children [WavePlayer_32]}
 {id=Modulator_8, children [Random_2], properties {pitchMax=0.9, pitchMin=0.9}}
 {id=Modulator_9, children [WavePlayer_9]}
 {id=Random_0, children [WavePlayer_0,WavePlayer_2], properties {weights [1,1]}}
 {id=Random_1, children [WavePlayer_12, WavePlayer_13], properties {weights [1,1]}}
 {id=Random_14, children [WavePlayer_30, WavePlayer_31], properties {weights [1,1]}}
 {id=Random_2, children [WavePlayer_5,WavePlayer_7], properties {weights [1,1]}}
 {id=Random_3, children [WavePlayer_8, WavePlayer_10], properties {weights [1,1]}}
 {id=Random_7, children [WavePlayer_15, WavePlayer_16], properties {weights [1,1]}}
 {id=WavePlayer_0, properties {soundWaveAssetPtr=/Figment_S08/Sounds/Fort_GamePlay_Sounds/Meteor/Explo_Meteor_Impact_01}}
 {id=WavePlayer_1, properties {soundWaveAssetPtr=/Game/Athena/Sounds/Ambient/Volcano/Amb_Volcano_FloorIsLava_Erupt}}
 {id=WavePlayer_10, properties {soundWaveAssetPtr=/Figment_S08/Sounds/Fort_GamePlay_Sounds/Meteor/Explo_Meteor_Impact_02}}
 {id=WavePlayer_11, properties {soundWaveAssetPtr=/Figment_S08/Sounds/Fort_GamePlay_Sounds/Meteor/Meteor_SubImpact_02}}
 {id=WavePlayer_12, properties {soundWaveAssetPtr=/Game/Sounds/Fort_GamePlay_Sounds/Meteor/Meteor_Almost_Landed_01}}
 {id=WavePlayer_13, properties {soundWaveAssetPtr=/Game/Sounds/Fort_GamePlay_Sounds/Meteor/Meteor_Almost_Landed_02}}
 {id=WavePlayer_14, properties {soundWaveAssetPtr=/Figment_S08/Sounds/Fort_GamePlay_Sounds/Meteor/Meteor_Sonic_Boom_02}}
 {id=WavePlayer_15, properties {soundWaveAssetPtr=/Figment_S08/Sounds/Fort_GamePlay_Sounds/Meteor/Meteor_SubImpact_03}}
 {id=WavePlayer_16, properties {soundWaveAssetPtr=/Figment_S08/Sounds/Fort_GamePlay_Sounds/Meteor/Meteor_SubImpact_01}}
 {id=WavePlayer_2, properties {soundWaveAssetPtr=&Explo_Meteor_Impact_02}}
 {id=WavePlayer_24, properties {soundWaveAssetPtr=&Meteor_Sonic_Boom_02}}
 {id=WavePlayer_25, properties {soundWaveAssetPtr=&Meteor_SubImpact_03}}
 {id=WavePlayer_27, properties {soundWaveAssetPtr=/Game/Sounds/Fort_GamePlay_Sounds/Explosives/fort_distant_explosion_sharp}}
 {id=WavePlayer_30, properties {soundWaveAssetPtr=&Meteor_Almost_Landed_01}}
 {id=WavePlayer_31, properties {soundWaveAssetPtr=&Meteor_Almost_Landed_02}}
 {
  id=WavePlayer_32
  properties {soundWaveAssetPtr=/Game/Sounds/Fort_GamePlay_Sounds/Explosives/ShockwaveGrenade/Grenade_Shockwave_ExplodeLayer01}
 }
 {id=WavePlayer_4, properties {soundWaveAssetPtr=&Meteor_SubImpact_03}}
 {id=WavePlayer_5, properties {soundWaveAssetPtr=&Explo_Meteor_Impact_01}}
 {id=WavePlayer_7, properties {soundWaveAssetPtr=&Explo_Meteor_Impact_02}}
 {id=WavePlayer_8, properties {soundWaveAssetPtr=&Explo_Meteor_Impact_01}}
 {id=WavePlayer_9, properties {soundWaveAssetPtr=&Meteor_Sonic_Boom_02}}
]
properties {
 attenuationSettings=/Figment_S08/Sounds/Attenuation/Fort_Volcano_Shoot_Far_Att
 maxDistance=2097152
 soundClassObject=/Figment_S08/Sounds/FortSoundClassesAndModes/Fort_Nexus_Transition
 volumeMultiplier=1
 concurrencyOverrides {maxCount=1, resolutionRule=PreventNew}
}
