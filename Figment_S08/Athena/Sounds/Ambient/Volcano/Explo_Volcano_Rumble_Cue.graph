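# Sound cue for a volcano explosion rumble, using a far volcano-shot attenuation and the Nexus transition sound class.
# A mixer layers a distant thunder rumble, played immediately, with a second thunder wave behind a fixed delay.
# An enveloper over the delayed layer holds volume near silence, then ramps it to full around the delayed start, so the thunder swells in over the rumble.
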SoundCue /Figment_S08/Athena/Sounds/Ambient/Volcano/Explo_Volcano_Rumble_Cue
root=Mixer_0
nodes [
 {id=Delay_2, children [WavePlayer_3], properties {delayMax=1.5, delayMin=1.5}}
 {
  id=Enveloper_0
  children [Delay_2]
  properties {
   pitchCurve {editorCurveData {keys [{time=0}]}}
   volumeCurve {editorCurveData {keys [<at 0 0.0006>,<at 0.1289 0.002>,<at 1.4114 0.0011>,<at 1.8518 0.1888>,<at 2.0297 0.4242>,<at 2.2161 1.016>]}}
  }
 }
 {id=Mixer_0, children [Enveloper_0,WavePlayer_5], properties {inputVolume [1,1]}}
 {id=WavePlayer_3, properties {soundWaveAssetPtr=&/Amb_Volcano_Thunder_04}}
 {id=WavePlayer_5, properties {soundWaveAssetPtr=/Game/Athena/Sounds/Ambient/Volcano/Amb_Volcano_Thunder_Rumble_Dist_01}}
]
properties {
 attenuationSettings=/Figment_S08/Sounds/Attenuation/Fort_Volcano_Shoot_Far_Att
 maxDistance=2097152
 soundClassObject=/Figment_S08/Sounds/FortSoundClassesAndModes/Fort_Nexus_Transition
 volumeMultiplier=1
 concurrencyOverrides {maxCount=1, resolutionRule=PreventNew}
}
