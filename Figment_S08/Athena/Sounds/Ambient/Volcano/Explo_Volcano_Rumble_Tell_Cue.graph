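# Volcano rumble "tell" sound cue, by its name a warning of an explosion, with far-distance attenuation.
# A mixer layers a modulated medium-distance rumble, randomly chosen earth-movement and thunder rumbles, and a boosted, modulated random distant growl.
# A random thunder clap enters after a short random delay beneath a rising volume envelope, and another distant growl follows a minute later.
# Concurrency allows one instance, rejecting new ones while it plays.
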SoundCue /Figment_S08/Athena/Sounds/Ambient/Volcano/Explo_Volcano_Rumble_Tell_Cue
root=Mixer_0
nodes [
 {id=Delay_0, children [WavePlayer_0], properties {delayMax=60, delayMin=60}}
 {id=Delay_2, children [Random_2], properties {delayMax=3, delayMin=2.5}}
 {
  id=Enveloper_0
  children [Delay_2]
  properties {
   pitchCurve {editorCurveData {keys [{time=0}]}}
   volumeCurve {editorCurveData {keys [<at 0 0.0006>,<at 0.1289 0.002>,<at 1.4114 0.0011>,<at 1.8518 0.1888>,<at 2.0297 0.4242>,<at 2.2161 1.016>]}}
  }
 }
 {id=Mixer_0, children [Modulator_0,Random_3,Random_4,Modulator_1,Enveloper_0,Delay_0], properties {inputVolume [1,0.5,0.35,2,1,1]}}
 {id=Modulator_0, children [WavePlayer_1]}
 {id=Modulator_1, children [Random_5]}
 {id=Random_2, children [WavePlayer_4,WavePlayer_2], properties {weights [1,1]}}
 {id=Random_3, children [WavePlayer_6,WavePlayer_7], properties {weights [1,1]}}
 {id=Random_4, children [WavePlayer_8, WavePlayer_9, WavePlayer_10], properties {weights [1,1,1]}}
 {id=Random_5, children [WavePlayer_11, WavePlayer_12], properties {weights [1,1]}}
 {id=WavePlayer_0, properties {soundWaveAssetPtr=&/Explo_Growl_Distant_02}}
 {id=WavePlayer_1, properties {soundWaveAssetPtr=&/Amb_Volcano_Thunder_Rumble_Med_Dist_02}}
 {id=WavePlayer_10, properties {soundWaveAssetPtr=&/Amb_Volcano_Thunder_Rumble_Dist_03}}
 {id=WavePlayer_11, properties {soundWaveAssetPtr=&/Explo_Growl_Distant_01}}
 {id=WavePlayer_12, properties {soundWaveAssetPtr=&Explo_Growl_Distant_02}}
 {id=WavePlayer_2, properties {soundWaveAssetPtr=/FNGameplayCues/Athena/Sounds/Ambient/Volcano/Amb_Volcano_Thunder_02}}
 {id=WavePlayer_4, properties {soundWaveAssetPtr=/FNGameplayCues/Athena/Sounds/Ambient/Volcano/Amb_Volcano_Thunder_03}}
 {id=WavePlayer_6, properties {soundWaveAssetPtr=/Game/Athena/Sounds/Ambient/Volcano/Amb_Volcano_Earth_Move_Rumble_Dist_05}}
 {id=WavePlayer_7, properties {soundWaveAssetPtr=/Game/Athena/Sounds/Ambient/Volcano/Amb_Volcano_Earth_Move_Rumble_Dist_04}}
 {id=WavePlayer_8, properties {soundWaveAssetPtr=/Game/Athena/Sounds/Ambient/Volcano/Amb_Volcano_Thunder_Rumble_Dist_01}}
 {id=WavePlayer_9, properties {soundWaveAssetPtr=/Game/Athena/Sounds/Ambient/Volcano/Amb_Volcano_Thunder_Rumble_Dist_02}}
]
properties {
 attenuationSettings=/Figment_S08/Sounds/Attenuation/Fort_Volcano_Shoot_Far_Att
 maxDistance=2097152
 soundClassObject=/Figment_S08/Sounds/FortSoundClassesAndModes/Fort_Nexus_Transition
 volumeMultiplier=0.5
 concurrencyOverrides {maxCount=1, resolutionRule=PreventNew}
}
