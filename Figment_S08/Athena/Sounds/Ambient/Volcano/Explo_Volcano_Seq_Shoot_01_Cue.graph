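# Sound cue for a volcano eruption "shoot" sequence, playing a timed series of rumbles and blasts through far-range volcano attenuation.
# A mixer starts the lava eruption wave at once and brings in two earth-move rumbles after a short delay, faded out by an envelope.
# A randomly chosen explosion follows shortly after, and a last explosion lands much later.
# Concurrency allows one instance and blocks new ones.
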SoundCue /Figment_S08/Athena/Sounds/Ambient/Volcano/Explo_Volcano_Seq_Shoot_01_Cue
root=Mixer_2
nodes [
 {id=Delay_2, children [Mixer_1], properties {delayMax=2, delayMin=2}}
 {id=Delay_3, children [Random_0], properties {delayMax=3, delayMin=3}}
 {id=Delay_4, children [WavePlayer_8], properties {delayMax=45, delayMin=45}}
 {
  id=Enveloper_0
  children [Delay_2]
  properties {pitchCurve {editorCurveData {keys [{time=0}]}}, volumeCurve {editorCurveData {keys [<at 0 1>,<at 9.1982 1>,<at 11.9406 -0.0067>]}}}
 }
 {id=Mixer_1, children [WavePlayer_2,WavePlayer_1], properties {inputVolume [1,1]}}
 {id=Mixer_2, children [WavePlayer_5,Enveloper_0,Delay_3,Delay_4], properties {inputVolume [1,1,1,1]}}
 {id=Random_0, children [WavePlayer_6,WavePlayer_7], properties {weights [1,1]}}
 {id=WavePlayer_1, properties {soundWaveAssetPtr=&/Amb_Volcano_Earth_Move_04}}
 {id=WavePlayer_2, properties {soundWaveAssetPtr=/FNGameplayCues/Athena/Sounds/Ambient/Volcano/Amb_Volcano_Earth_Move_01}}
 {id=WavePlayer_5, properties {soundWaveAssetPtr=/Game/Athena/Sounds/Ambient/Volcano/Amb_Volcano_FloorIsLava_Erupt}}
 {id=WavePlayer_6, properties {soundWaveAssetPtr=/Game/Sounds/Fort_GamePlay_Sounds/Explosives/bomb_explode_stereo_01}}
 {id=WavePlayer_7, properties {soundWaveAssetPtr=/Game/Sounds/Fort_GamePlay_Sounds/Explosives/bomb_explode_stereo_02}}
 {id=WavePlayer_8, properties {soundWaveAssetPtr=&bomb_explode_stereo_01}}
]
properties {
 attenuationSettings=/Figment_S08/Sounds/Attenuation/Fort_Volcano_Shoot_Far_Att
 maxDistance=2097152
 soundClassObject=/Figment_S08/Sounds/FortSoundClassesAndModes/Fort_Nexus_Transition
 volumeMultiplier=1
 concurrencyOverrides {maxCount=1, resolutionRule=PreventNew}
}
