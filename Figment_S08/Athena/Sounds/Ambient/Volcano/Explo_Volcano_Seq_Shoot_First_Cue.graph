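# Sound cue for the first shot in a volcano explosion sequence, built for very long-range audibility.
# A mixer layers an immediate lava eruption sound with two fixed-delay layers: a boosted snow-event ending sound seconds later, and a stereo bomb explosion a minute in.
# It uses far volcano-shoot attenuation and the Nexus transition sound class, and concurrency allows a single instance, blocking new ones while it plays.
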SoundCue /Figment_S08/Athena/Sounds/Ambient/Volcano/Explo_Volcano_Seq_Shoot_First_Cue
root=Mixer_2
nodes [
 {id=Delay_3, children [WavePlayer_0], properties {delayMax=3, delayMin=3}}
 {id=Delay_4, children [WavePlayer_8], properties {delayMax=60, delayMin=60}}
 {id=Mixer_2, children [WavePlayer_5,Delay_3,Delay_4], properties {inputVolume [1,1.75,1]}}
 {id=WavePlayer_0, properties {soundWaveAssetPtr=/Figment_S08/Athena/Sounds/Events/Dopey/Snow/Snow_End_02}}
 {id=WavePlayer_5, properties {soundWaveAssetPtr=/Game/Athena/Sounds/Ambient/Volcano/Amb_Volcano_FloorIsLava_Erupt}}
 {id=WavePlayer_8, properties {soundWaveAssetPtr=/Game/Sounds/Fort_GamePlay_Sounds/Explosives/bomb_explode_stereo_01}}
]
properties {
 attenuationSettings=/Figment_S08/Sounds/Attenuation/Fort_Volcano_Shoot_Far_Att
 maxDistance=2097152
 soundClassObject=/Figment_S08/Sounds/FortSoundClassesAndModes/Fort_Nexus_Transition
 volumeMultiplier=1
 concurrencyOverrides {maxCount=1, resolutionRule=PreventNew}
}
