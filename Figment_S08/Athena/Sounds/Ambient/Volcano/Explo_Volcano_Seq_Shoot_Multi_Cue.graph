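# Sound cue for a volcano eruption sequence, sharing one far-distance attenuation and allowing a single instance at a time.
# A top mixer plays the lava eruption wave immediately alongside two delayed branches.
# The first, after a short delay, layers two earth-movement rumbles with a slightly later modulated explosion, all under an envelope that holds, then fades out.
# The second fires a lone stereo explosion much later, fifty seconds in.
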SoundCue /Figment_S08/Athena/Sounds/Ambient/Volcano/Explo_Volcano_Seq_Shoot_Multi_Cue
root=Mixer_2
nodes [
 {id=Delay_2, children [Mixer_1], properties {delayMax=2.5, delayMin=2.5}}
 {id=Delay_3, children [Modulator_1], properties {delayMax=0.1, delayMin=0.1}}
 {id=Delay_6, children [WavePlayer_23], properties {delayMax=50, delayMin=50}}
 {
  id=Enveloper_0
  children [Delay_2]
  properties {pitchCurve {editorCurveData {keys [{time=0}]}}, volumeCurve {editorCurveData {keys [<at 0 1>,<at 9.1982 1>,<at 11.9406 -0.0067>]}}}
 }
 {id=Mixer_1, children [WavePlayer_2,WavePlayer_1,Delay_3], properties {inputVolume [1,1,1]}}
 {id=Mixer_2, children [WavePlayer_5,Enveloper_0,Delay_6], properties {inputVolume [1,1,1]}}
 {id=Modulator_1, children [WavePlayer_4]}
 {id=WavePlayer_1, properties {soundWaveAssetPtr=&/Amb_Volcano_Earth_Move_04}}
 {id=WavePlayer_2, properties {soundWaveAssetPtr=/FNGameplayCues/Athena/Sounds/Ambient/Volcano/Amb_Volcano_Earth_Move_01}}
 {id=WavePlayer_23, properties {soundWaveAssetPtr=/Game/Sounds/Fort_GamePlay_Sounds/Explosives/bomb_explode_stereo_01}}
 {id=WavePlayer_4, properties {soundWaveAssetPtr=&bomb_explode_stereo_01}}
 {id=WavePlayer_5, properties {soundWaveAssetPtr=/Game/Athena/Sounds/Ambient/Volcano/Amb_Volcano_FloorIsLava_Erupt}}
]
properties {
 attenuationSettings=/Figment_S08/Sounds/Attenuation/Fort_Volcano_Shoot_Far_Att
 maxDistance=2097152
 soundClassObject=/Figment_S08/Sounds/FortSoundClassesAndModes/Fort_Nexus_Transition
 volumeMultiplier=1
 concurrencyOverrides {maxCount=1, resolutionRule=PreventNew}
}
